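# Physics asset for the Farmer face accessory, covering its dynamic hat and left and right ponytail bones.
# Capsule bodies on head, neck, upper spine, clavicles and upper arms form an upper-body proxy, joined by free-rotation constraints.
# Each ponytail is a three-link chain from the head with tight twist and differing left and right swing limits.
# The hat is one wide capsule joined to the head with free rotation.
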PhysicsAsset /BRCosmetics/Accessories/Hats/F_MED_Farmer_FaceAcc/Meshes/F_MED_Farmer_FaceAcc_Dyn
bodies {
 clavicle_l [{shapeType=Capsule, length=7.6738, radius=6.8445, center=<4.848 -1.4292 -1.8106>, rotation=<rot 74.9997 0 0>}]
 clavicle_r [{shapeType=Capsule, length=6.9863, radius=7.532, center=<-4.7278 1.5741 2.2594>, rotation=<rot 74.9997 0 0>}]
 dyn_hat [{shapeType=Capsule, length=1.875, radius=21.4626, center=<13.2886 -5.7062 0.0006>, rotation=<rot -0.0003 -20.0002 89.9991>}]
 dyn_ponytail_l_01 [{shapeType=Capsule, length=3.25, radius=1.75, center=<3 0 0>, rotation=<rot 90 0 0>}]
 dyn_ponytail_l_02 [{shapeType=Capsule, length=6.0625, radius=1.9062, center=<4.0706 -0.0001 0>, rotation=<rot 90 0 0>}]
 dyn_ponytail_l_03 [{shapeType=Capsule, length=4.0382, radius=2.4305, center=<3.8231 -0.3598 -0.8757>, rotation=<rot 90 0 0>}]
 dyn_ponytail_r_01 [{shapeType=Capsule, length=3.25, radius=1.75, center=<-3 0 0>, rotation=<rot 90 0 0>}]
 dyn_ponytail_r_02 [{shapeType=Capsule, length=6.0312, radius=1.8438, center=<-4.0997 0 0>, rotation=<rot 90 0 0>}]
 dyn_ponytail_r_03 [{shapeType=Capsule, length=6.1623, radius=1.7566, center=<-4.3831 0.2316 0.3691>, rotation=<rot 90 7.5952 0>}]
 head [
  {shapeType=Capsule, length=2.85, radius=5.7337, center=<1.5439 3.4489 0.0001>, rotation=<rot -89.5256 -89.7443 90.2186>}
  {shapeType=Capsule, length=4.875, radius=3.625, center=<-2.8878 -8.8048 -6.3236>, rotation=<rot -90 0 -18.7649>}
 ]
 neck_02 [{shapeType=Capsule, length=10.9843, radius=4.6791, center=<0 -0.3727 0>, rotation=<rot 0 -89.9999 -89.9996>}]
 spine_05 [
  {shapeType=Capsule, length=12.8619, radius=9.6069, center=<1.1686 -2.6897 0>}
  {shapeType=Capsule, length=7.4017, radius=7.0045, center=<5.2338 1.9611 -0.0001>, rotation=<rot 90 -14.0363 9.1129>}
 ]
 upperarm_l [{shapeType=Capsule, length=20.741, radius=4.1987, center=<11.4355 0 0>, rotation=<rot 90 0 0>}]
 upperarm_r [{shapeType=Capsule, length=20.741, radius=4.1987, center=<-11.4355 0 0>, rotation=<rot 90 0 0>}]
}
constraints [
 {
  bone1Name=clavicle_r
  bone2Name=spine_05
  swing1LimitDegrees=45
  swing1Motion=Free
  swing2LimitDegrees=45
  swing2Motion=Free
  twistLimitDegrees=45
  twistMotion=Free
 }
 {
  bone1Name=clavicle_l
  bone2Name=spine_05
  swing1LimitDegrees=45
  swing1Motion=Free
  swing2LimitDegrees=45
  swing2Motion=Free
  twistLimitDegrees=45
  twistMotion=Free
 }
 {
  bone1Name=neck_02
  bone2Name=spine_05
  swing1LimitDegrees=45
  swing1Motion=Free
  swing2LimitDegrees=45
  swing2Motion=Free
  twistLimitDegrees=45
  twistMotion=Free
 }
 {
  bone1Name=head
  bone2Name=neck_02
  swing1LimitDegrees=45
  swing1Motion=Free
  swing2LimitDegrees=45
  swing2Motion=Free
  twistLimitDegrees=45
  twistMotion=Free
 }
 {bone1Name=dyn_ponytail_r_01, bone2Name=head, swing1LimitDegrees=30, swing2LimitDegrees=60, twistLimitDegrees=5}
 {bone1Name=dyn_ponytail_r_02, bone2Name=dyn_ponytail_r_01, swing1LimitDegrees=30, swing2LimitDegrees=60, twistLimitDegrees=5}
 {bone1Name=dyn_ponytail_r_03, bone2Name=dyn_ponytail_r_02, swing1LimitDegrees=30, swing2LimitDegrees=60, twistLimitDegrees=5}
 {bone1Name=dyn_ponytail_l_01, bone2Name=head, swing1LimitDegrees=20, swing2LimitDegrees=20, twistLimitDegrees=5}
 {bone1Name=dyn_ponytail_l_02, bone2Name=dyn_ponytail_l_01, swing1LimitDegrees=60, swing2LimitDegrees=60, twistLimitDegrees=5}
 {bone1Name=dyn_ponytail_l_03, bone2Name=dyn_ponytail_l_02, swing1LimitDegrees=75, swing2LimitDegrees=75, twistLimitDegrees=5}
 {
  bone1Name=upperarm_l
  bone2Name=clavicle_l
  swing1LimitDegrees=45
  swing1Motion=Free
  swing2LimitDegrees=45
  swing2Motion=Free
  twistLimitDegrees=45
  twistMotion=Free
 }
 {
  bone1Name=upperarm_r
  bone2Name=clavicle_r
  swing1LimitDegrees=45
  swing1Motion=Free
  swing2LimitDegrees=45
  swing2Motion=Free
  twistLimitDegrees=45
  twistMotion=Free
 }
 {
  bone1Name=dyn_hat
  bone2Name=head
  swing1LimitDegrees=45
  swing1Motion=Free
  swing2LimitDegrees=45
  swing2Motion=Free
  twistLimitDegrees=45
  twistMotion=Free
 }
]
