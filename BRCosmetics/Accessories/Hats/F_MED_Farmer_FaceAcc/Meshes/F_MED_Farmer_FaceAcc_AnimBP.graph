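# Animation Blueprint for the Farmer hat face accessory, animating the hat and its ponytails on the character.
# The live pose chain copies the attached parent's pose, simulates a RigidBody physics asset in head-bone space with custom gravity, then applies small additive ponytail rotations.
# Leftover hat-lift logic maps brim-to-shoulder-target distance to `head_up` for additive hat and brim bends, but those bone nodes and the update event are unwired.
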AnimBlueprint /BRCosmetics/Accessories/Hats/F_MED_Farmer_FaceAcc/Meshes/F_MED_Farmer_FaceAcc_AnimBP
target_skeleton=&/F_MED_Farmer_FaceAcc_Skeleton
graphs [
 AnimGraph {nodes [
   n1 = Get head_up() -> (head_up:double)
   "Transform (Modify) Bone - Bone: dyn_hat_brim_back"(Alpha=@n1, Translation="0,0,0", Rotation="-17.025,65.155,-4.542", Scale="1,1,1") {
    rotationMode=BMM_Additive
    rotationSpace=BCS_BoneSpace
    alphaBoolBlend {blendOption=Linear}
    alphaScaleBiasClamp {bInterpResult=true, interpSpeedDecreasing=30, interpSpeedIncreasing=30}
    boneToModify {boneName=dyn_hat_brim_back}
    rotation=<rot -17.025 65.1551 -4.542>
   }
   "Transform (Modify) Bone - Bone: dyn_hat"(Alpha=@n1, Translation="0.132,0.136,-0.136", Rotation="0.000,20,0.000", Scale="1,1,1") {
    rotationMode=BMM_Additive
    rotationSpace=BCS_BoneSpace
    translationMode=BMM_Additive
    translationSpace=BCS_BoneSpace
    alphaBoolBlend {blendOption=Linear}
    alphaScaleBiasClamp {bInterpResult=true, interpSpeedDecreasing=20, interpSpeedIncreasing=20}
    boneToModify {boneName=dyn_hat}
    rotation {yaw=20}
    translation=<0.132 0.136 -0.136>
   }
   n2 = RigidBody(ComponentPose=Local To Component(LocalPose=Copy Pose From Mesh() {bUseAttachedParent=true})) {
    bOverrideWorldGravity=true
    overridePhysicsAsset=&/F_MED_Farmer_FaceAcc_Dyn
    simulationSpace=BaseBoneSpace
    alphaBoolBlend {blendOption=Linear}
    baseBoneRef {boneName=head}
    componentAppliedLinearAccClamp=<2000 2000 2000>
    componentLinearAccScale=<0.2 0.5 0.8>
    componentLinearVelScale=<0.2 0.5 0.8>
    overrideWorldGravity {z=-980}
   }
   n3 = "Transform (Modify) Bone - Bone: dyn_ponytail_r_03"!(ComponentPose=@n2, bAlphaBoolEnabled:bool=True, Translation="0,0,0", Rotation="-3,0,0", Scale="1,1,1") -> (Pose) {alphaInputType=Bool, rotationMode=BMM_Additive, rotationSpace=BCS_BoneSpace, boneToModify {boneName=dyn_ponytail_r_03}, rotation {pitch=-3}}
   n4 = "Transform (Modify) Bone - Bone: dyn_ponytail_l_03"!(ComponentPose=@n3, bAlphaBoolEnabled:bool=True, Translation="0,0,0", Rotation="-5,0,0", Scale="1,1,1") -> (Pose) {alphaInputType=Bool, rotationMode=BMM_Additive, rotationSpace=BCS_BoneSpace, boneToModify {boneName=dyn_ponytail_l_03}, rotation {pitch=-5}}
   Output Pose(Result=Component To Local(ComponentPose=@n4)) {layerGroup=None}
  ]}
 EventGraph {nodes [
   n1 = Branch(Condition=Get bIsTargeting())
   n2 = GetPartSkeletalMeshComponent()
   > Set head_up(head_up:double=MapRangeClamped(
     Value=Get Distance Between Two Sockets!(
      Component:Skeletal Mesh Component Object Reference=@n2
      SocketOrBoneNameA:Name=dyn_hat_brim_back
      SocketSpaceA:ERelativeTransformSpace Enum=RTS_World
      SocketOrBoneNameB:Name=dyn_hat_shoulder_target
      SocketSpaceB:ERelativeTransformSpace Enum=RTS_World
      bRemapRange:bool=false
      InRangeMin:float=0
      InRangeMax:float=0
      OutRangeMin:float=0
      OutRangeMax:float=0
     ) -> (ReturnValue:float) {is=title K2Node_CallFunction}
     InRangeA=8.1
     InRangeB=7
    )) -> (Output_Get:double)
   Set head_up(execute=@n1.else, head_up=0)
   n3 = GetDeltaTransformFromRefPose(self=GetOwnerSkeletalMeshComponent(), BoneName=head, BaseName=spine_05)
   MapRangeClamped(Value=BreakRotator(InRot=BreakTransform(InTransform=@n3).Rotation).Yaw, InRangeA=0, InRangeB=-40)
   TryGetPawnOwner()
   Event BlueprintUpdateAnimation()
  ]}
]
variables [head_up {type=number}]
